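; Lowrider hydraulic mod world object (Valet "Slinky") that Athena players pick up as an item or use as a bounce pad.
; JumpCollision launches pawns with ceiling-checked heights, vehicles through OnRep_OverlapVehicle, and hoverboards, with material-driven jiggle animation.
; Projectile stop/bounce applies the mod to eligible Valet vehicles (ShouldApplyMod returns true to block it) or can spawn a pickup, as does water entry.
; Losing its supporting actor restarts movement.
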(event EventBeginPlay
  ;[1 0.9894 0] Begin Play
  (|CallParentFunction)
  (Variables|Default|SetJumpCollisionLocation (+ (Transformation|GetActorLocation) (Variables|Default|GetJump%20Offset))) ; Do this so we have the same location on server and client
  (Networking|SwitchHasAuthority
    (:Authority
      (Utilities|FlowControl|Sequence
        (:then_0
          (Variables|Interaction|SetbAllowInteract false)
          (Utilities|FlowControl|DelayUntilNextTick)
          (Variables|Interaction|SetbAllowInteract true))
        (:then_1
          ;;[0 1 0.6353] Random Rotation
          (Utilities|FlowControl|DoOnce false
            (:Completed
              (Transformation|SetRelativeRotation (Variables|Default|GetTire) (Math|Rotator|MakeRotator 0.0 0.0 (Math|Random|RandomFloatinRange -1.8e+02 1.8e+02)) false false)
              ;;[0 1 0.6353] Spawn Animation
              (CallFunction|MIDPlayAnimation 0.0 2e+01)
              (CallFunction|MulticastPlaySpawnSound))) ; doonce because begin play fires off all the time.
        )))))

(event Custom|Multicast_Play%20Spawn%20Sound
  ;[1 0.9894 0] Begin Play
  ;;[0.06227 0.06198 1] Spawn Sound plays to call clients when server spawns
  (Audio|PlaySoundatLocation "/ValetMods/Audio/OffRoadTire/ValetMod_OffRoadTire_Pile_Spawn_Cue.ValetMod_OffRoadTire_Pile_Spawn_Cue" (Transformation|GetActorLocation)))

(event Custom|Multicast_HideAndDestroy
  ;[0.973 0 1] Hide and Destroy
  (Rendering|SetActorHiddenInGame true)
  (Actor|DestroyActor))

(event EventOnDeathServer
  (params (Damage :type Float%20%28single-precision%29) (DamageTags :type Gameplay%20Tag%20Container%20Structure%20%28by%20ref%29) (Momentum :type Vector :default "0, 0, 0") (HitInfo :type Hit%20Result%20Structure%20%28by%20ref%29) (InstigatedBy :type Controller%20Object%20Reference) (DamageCauser :type Actor%20Object%20Reference) (EffectContext :type Gameplay%20Effect%20Context%20Handle%20Structure))
  ;[0.973 0 1] Hide and Destroy
  (Networking|FlushNetDormancy)
  (Rendering|SetActorHiddenInGame true))

(event Interaction|EventBlueprintOnInteract
  (params (InteractingPawn :type Fort%20Pawn%20Object%20Reference) (InteractionBeingAttempted :type EInteractionBeingAttempted%20Enum :default "FirstInteraction"))
  ;[0.9534 1 0] EVENT -> Interact
  (if (Utilities|Enum|Equal(Enum) (Fortnite|GetSubGame) "Athena")
    (bind _asfort_player_pawn (Utilities|Casting|CastToFortPlayerPawn InteractingPawn))
    (bind _output_get (Variables|Default|SetPlayerPawn _asfort_player_pawn))
    (bind _asfort_player_controller_athena (Utilities|Casting|CastToFortPlayerControllerAthena (Fortnite|GetFortPlayerControllerfromActor _output_get)))
    (Variables|Default|SetPlayerController _asfort_player_controller_athena)
    (Ability|ExecuteGameplayCueWithParamsReplicated (Variables|GameplayCues|GetGC_Pickup) (Ability|GameplayCue|MakeGameplayCueParameters :NormalizedMagnitude 0.0 :RawMagnitude 0.0 :Location (Transformation|GetActorLocation) :Normal (Transformation|GetActorForwardVector) :GameplayEffectLevel 1 :AbilityLevel (Math|Conversions|ToInteger(Byte) (Variables|GameplayCues|GetObjectWeights)) :bReplicateLocationWhenUsingMinimalRepProxy false))
    (if (Class|FortItemEntryComponent|HasItem (Variables|Default|GetFortItemEntry))
      (Class|FortItemEntryComponent|AddItemtoInventoryOwner (Variables|Default|GetFortItemEntry) (Variables|Default|GetPlayerController) true)
      (else
        (Inventory|GiveItemtoInventoryOwner :InventoryOwner (Variables|Default|GetPlayerController) :ItemDefinition (Variables|Default|GetWeaponItemDefinition) :NumberToGive 1 :bNotifyPlayer true :ItemLevel -1 :PickupInstigatorHandle -1 :bUseItemPickupAnalyticEvent true :WeaponAmmoOverride -1)))
    (CallFunction|MulticastHideandDestroy)))

(event OnComponentBeginOverlap(JumpCollision)
  (params (OverlappedComponent :type Primitive%20Component%20Object%20Reference) (OtherActor :type Actor%20Object%20Reference) (OtherComp :type Primitive%20Component%20Object%20Reference) (OtherBodyIndex :type Integer) (bFromSweep :type Boolean) (SweepResult :type Hit%20Result%20Structure%20%28by%20ref%29))
  (bind _overlappawn (Variables|Default|GetOverlapPawn))
  ;[0 0.1508 1] Bounce off BGA
  (Networking|FlushNetDormancy)
  (Variables|Default|SetJumpCollisionLocation (+ (Transformation|GetActorLocation) (Variables|Default|GetJump%20Offset))) ; Do this so we have the same location on server and client
  (bind _asfort_pawn (Utilities|Casting|CastToFortPawn OtherActor)
    (:then
      (if (not (Utilities|Equal(Object) (Variables|Default|GetOverlapPawn) _asfort_pawn))
        (bind _output_get (Variables|Default|SetOverlapPawn _asfort_pawn))
        (if (or (GameplayTags|HasMatchingGameplayTag (Ability|GetAbilitySystemComponent _output_get) "(TagName=\"Gameplay.Action.Player.Launched\")") (Variables|/Script/FortniteGame.FortPawn|GetbIsDBNO _output_get))
          (Variables|Default|SetOverlapPawn 0) ; Clear Overlap Pawn if true. Fixes FORT-366067
          (else
            (bind _asfort_player_controller (Utilities|Casting|CastToFortPlayerController (Pawn|GetController _output_get))
              (:then
                (Variables|Default|SetOverlapController _asfort_player_controller))
              (:CastFailed
                _))
            (CallFunction|DetermineBounceHeights)
            (if (> (ScalableFloat|GetValueatLevel (Variables|Default|GetShouldImpulsePlayer)) 0.0)
              (Character|LaunchCharacter (Variables|Default|GetOverlapPawn) (Math|Vector|MakeVector 0.0 0.0 (Variables|Default|GetFinalBounceJumpHeight)) false true)
              (GameplayEffects|ApplyGameplayEffectToSelf :self (Ability|GetAbilitySystemComponent _overlappawn) :GameplayEffectClass (Variables|Default|GetGE_PlayerLaunched) :Level 0.0)
              (Variables|/Script/FortniteGame.FortPawn|SetbIgnoreNextFallingDamage true _overlappawn)
              (CallFunction|MulticastPlayerBounceCosmetics))))))
    (:CastFailed
      (bind _asfort_athena_vehicle (Utilities|Casting|CastToFortAthenaVehicle OtherActor)
        (:then
          (Variables|Default|SetOverlapVehicle _asfort_athena_vehicle))
        (:CastFailed
          ;;[1 1 1] Simulating Physics Object - Destroy self
          (bind _returnvalue (FortPhysicsSimulationLibrary|IsPhysicsObjectActiveinSimulation OtherActor))
          (if _returnvalue
            (CallFunction|PlayGenericDeath)
            (Actor|DestroyActor)
            (else
              (CallFunction|DetermineBounceHeights)
              (bind _success (Vehicles|LaunchHoverboard OtherActor (Math|Vector|MakeVector 0.0 0.0 (Variables|Default|GetFinalBounceJumpHeight)) "TireLaunch"))
              (if _success
                (Ability|ExecuteGameplayCueReplicated (Variables|Default|GetGC%20Bounce%20Player))
                ;;[1 0.2874 0] Bounce Animation
                (CallFunction|MIDPlayAnimation 3e+01 54.0)))))))))

(event Custom|LaunchVehicle
  (bind _overlapvehicle (Variables|Default|GetOverlapVehicle))
  ;[0 0.1508 1] Bounce off BGA
  ;;[0 0 0] Launch Vehicles
  (Audio|BroadcastAbilityImpactatLocation 0 _overlapvehicle (CallFunction|CalcVehicleImpulse _overlapvehicle (ScalableFloat|GetValueatLevel (Variables|Default|GetHotfixVehicleImpulse))) (Variables|Default|GetVehicle%20Impulse%20Radius))
  (Audio|BroadcastAbilitySpinThrustatLocation 0 _overlapvehicle (* (Math|Vector|SelectVector (Math|Vector|MirrorVectorbyNormal (Vehicle|GetVehicleRightVector _overlapvehicle) (Vehicle|GetVehicleRightVector _overlapvehicle)) (Vehicle|GetVehicleRightVector _overlapvehicle) (> (Vehicle|GetForwardSpeedKmh _overlapvehicle) 0.0)) (ScalableFloat|GetValueatLevel (Variables|Default|GetHotfixSpinThrust))) (Variables|Default|GetVehicle%20Impulse%20Radius))
  (Networking|CanExecuteCosmeticEvents
    (:False
      (Ability|ExecuteGameplayCueReplicated (Variables|Default|GetGC_BounceVehicle)))) ; Prevents GC from playing multiple times.
)

(event Custom|Multicast_PlayerBounceCosmetics
  (bind _overlapcontroller (Variables|Default|GetOverlapController))
  ;[0 0.1508 1] Bounce off BGA
  ;;[0.7558 1 0] Player Bounce Cosmetics
  (CallFunction|MIDPlayAnimation 3e+01 54.0)
  ;;;[1 1 1] Bounce Effect
  (Ability|ExecuteGameplayCueReplicated (Variables|Default|GetGC%20Bounce%20Player))
  (Utilities|IsValid _overlapcontroller
    (:Is%20Valid
      (Game|Feedback|LocalPlayForceFeedback _overlapcontroller (Variables|Default|GetLaunchForceFeedback) false "TireLaunch"))
    (:Is%20Not%20Valid
      _))
  ;;;[1 1 1] Delay so player has time to bounce away
  (Utilities|FlowControl|Delay (Variables|Default|GetPlayerBounceDelay))
  (Variables|Default|SetOverlapPawn 0)
  (Variables|Default|SetOverlapController 0))

(event OnProjectileStop(FortProjectileMovement)
  (params (ImpactResult :type Hit%20Result%20Structure%20%28by%20ref%29))
  (bind (_bblockinghit _binitialoverlap _time _distance _location _impactpoint _normal _impactnormal _physmat _hitactor) (Collision|BreakHitResult ImpactResult))
  (bind _hit_actor (Variables|Default|GetHit%20Actor))
  (bind (_isvalet _valetvehicle) (CallFunction|IsValetVehicle _hit_actor))
  (bind (_bblockinghit_2 _binitialoverlap_2 _time_2 _distance_2 _location_2 _impactpoint_2 _normal_2 _impactnormal_2 _physmat_2 _hitactor_2) (Collision|BreakHitResult ImpactResult))
  (bind (_asfort_vehicle_interface _bsuccess) (Utilities|Casting|CastToFortVehicleInterface _hitactor_2))
  ;[0 0.5 0.02966] On Projectile Stop: Start moving again if the object we are resting on is destroyed.
  (Networking|FlushNetDormancy)
  (Variables|Default|SetJumpCollisionLocation (+ (Transformation|GetActorLocation) (Variables|Default|GetJump%20Offset))) ; Do this so we have the same location on server and client
  (Networking|SwitchHasAuthority
    (:Authority
      (Variables|ProjectileMovement|SetRepStartMoving 0)
      (Variables|Default|SetHit%20Actor _hitactor_2)
      (if (not _bsuccess)
        ;;[0 0 0] If the structure we're sitting on dies then we should start moving again
        (bind _asbuilding_actor (Utilities|Casting|CastToBuildingActor _hitactor)
          (:then
            (EventDispatchers|BindEventtoOnDied _asbuilding_actor (EventDispatchers|CreateEvent :!event Custom|EventAttachedBADied)))
          (:CastFailed
            (Utilities|Casting|CastToFortPlayerPawn _hitactor
              (:then
                (CallFunction|SpawnPickUp))
              (:CastFailed
                (Game|BindEventtoOnDestroyed _hitactor (EventDispatchers|CreateEvent :!event Custom|EventAttachedActorDestroyed)))) ; if we land on a player then spawn pickup
          ))
        (elif _isvalet
          ;[0 1 1] Apply Tire Mod if we hit a vehicle while falling.
          (if (CallFunction|ShouldApplyMod _valetvehicle)
            (CallFunction|SpawnPickUp) ; If we shouldn't apply mod, or we hit a non Valet vehicle then spawn a pickup to avoid bouncy crazyness
            (else
              (Utilities|FlowControl|DoOnce false
                (:Completed
                  (Networking|SwitchHasAuthority
                    (:Authority
                      (Vehicle|Mod|ApplyVehicleMod _valetvehicle (Variables|Default|GetModTagToApply) 0) ; No instigator\n(Used instead of ApplyTireMod for proper internal setup)
                      (Rendering|SetVisibility (Variables|Default|GetTire) false true)
                      (Actor|DestroyActor)))))))
          (else
            (Utilities|Casting|CastToFortVehicleInterface (Variables|Default|GetHit%20Actor)) ; Did we hit a vehicle other than Valet
          ))))))

(event Custom|EventAttachedBADied
  (params (DamagedActor :type Actor%20Object%20Reference) (Damage :type Float%20%28single-precision%29) (InstigatedBy :type Controller%20Object%20Reference) (DamageCauser :type Actor%20Object%20Reference) (HitLocation :type Vector) (FHitComponent :type Primitive%20Component%20Object%20Reference) (BoneName :type Name) (Momentum :type Vector))
  ;[0 0 0] bound events to if the structure we're sitting on dies then we should start moving again
  (Networking|FlushNetDormancy)
  (Variables|ProjectileMovement|SetRepStartMoving 1) ; Get us moving again
)

(event OnProjectileBounce(FortProjectileMovement)
  (params (ImpactResult_bBlockingHit :type Boolean) (ImpactResult_bInitialOverlap :type Boolean) (ImpactResult_Time :type Float%20%28single-precision%29) (ImpactResult_Distance :type Float%20%28single-precision%29) (ImpactResult_Location :type Vector :default "0, 0, 0") (ImpactResult_ImpactPoint :type Vector :default "0, 0, 0") (ImpactResult_Normal :type Vector :default "0, 0, 0") (ImpactResult_ImpactNormal :type Vector :default "0, 0, 0") (ImpactResult_PhysMat :type Physical%20Material%20Object%20Reference) (ImpactResult_HitActor :type Actor%20Object%20Reference) (ImpactResult_HitComponent :type Primitive%20Component%20Object%20Reference) (ImpactResult_HitBoneName :type Name :default "None") (ImpactResult_BoneName :type Name :default "None") (ImpactResult_HitItem :type Integer) (ImpactResult_ElementIndex :type Integer) (ImpactResult_FaceIndex :type Integer) (ImpactResult_TraceStart :type Vector :default "0, 0, 0") (ImpactResult_TraceEnd :type Vector :default "0, 0, 0") (ImpactVelocity :type Vector%20%28by%20ref%29 :default "0, 0, 0"))
  (bind _hit_actor (Variables|Default|GetHit%20Actor))
  (bind (_isvalet _valetvehicle) (CallFunction|IsValetVehicle _hit_actor))
  ;[1 0 0.8328] On Actor Bounce
  (Variables|Default|SetHit%20Actor ImpactResult_HitActor)
  ;[0 1 1] Apply Tire Mod if we hit a vehicle while falling.
  (if _isvalet
    (if (not (CallFunction|ShouldApplyMod _valetvehicle))
      (Utilities|FlowControl|DoOnce false
        (:Completed
          (Networking|SwitchHasAuthority
            (:Authority
              (Vehicle|Mod|ApplyVehicleMod _valetvehicle (Variables|Default|GetModTagToApply) 0) ; No instigator\n(Used instead of ApplyTireMod for proper internal setup)
              (Rendering|SetVisibility (Variables|Default|GetTire) false true)
              (Actor|DestroyActor))))))
    (else
      (Utilities|Casting|CastToFortVehicleInterface (Variables|Default|GetHit%20Actor)) ; Did we hit a vehicle other than Valet
    ))
  (CallFunction|SpawnPickUp) ; If we shouldn't apply mod, or we hit a non Valet vehicle then spawn a pickup to avoid bouncy crazyness
)

(event Custom|EventAttachedActorDestroyed
  (params (DestroyedActor :type Actor%20Object%20Reference))
  ;[0 0 0] bound events to if the structure we're sitting on dies then we should start moving again
  (Networking|FlushNetDormancy)
  (Variables|ProjectileMovement|SetRepStartMoving 1) ; Get us moving again
)

(event OnComponentEndOverlap(JumpCollision)
  (params (OverlappedComponent :type Primitive%20Component%20Object%20Reference) (OtherActor :type Actor%20Object%20Reference) (OtherComp :type Primitive%20Component%20Object%20Reference) (OtherBodyIndex :type Integer))
  ;[1 0.7175 0] If a tire mod moves away from us then we probably need to start simulating again!
  (Networking|SwitchHasAuthority
    (:Authority
      (bind _asbga_clyde_valet_mod_slinky_pickup (Utilities|Casting|CastToBGA_Clyde_ValetMod_Slinky_Pickup OtherActor))
      (if (!= (Variables|/Game/Athena/Items/HeldObjects/HeldObjectVersion2/BGA_HeldObject_Parent.BGA_HeldObject_Parent_C|GetRepStartMoving _asbga_clyde_valet_mod_slinky_pickup) 1)
        (Networking|FlushNetDormancy)
        (Variables|/Game/Athena/Items/HeldObjects/HeldObjectVersion2/BGA_HeldObject_Parent.BGA_HeldObject_Parent_C|SetRepStartMoving 1 _asbga_clyde_valet_mod_slinky_pickup)))))

(event OnEnterWaterEvent(FortWaterInteraction)
  (params (WaterBody :type Fort%20Water%20Body%20Actor%20Object%20Reference) (WaterInteractionComponent :type Fort%20Water%20Interaction%20Component%20Object%20Reference) (bIsFirstBody :type Boolean))
  ;[0 0.1767 1] If we fall into a water body then spawn a pickup (Fixes FORT-361564)
  (Effects|Components|ParticleSystem|SpawnFXAtLocation (Variables|Default|GetFX_WaterSpash) (Transformation|GetActorLocation))
  (CallFunction|SpawnPickUp))

(event Custom|SpawnPickUp
  (bind (_returnvalue_x _returnvalue_y _returnvalue_z) (Math|Vector|BreakVector (Transformation|GetActorLocation)))
  ;[0 0.4271 0.07421] Spawn Pickup; Changed from function to event because "do once" was triggering multiple times inside a function
  (Utilities|FlowControl|DoOnce false
    (:Completed
      (Networking|SwitchHasAuthority
        (:Authority
          (Rendering|SetActorHiddenInGame true)
          (Inventory|SpawnPickupInWorld (Variables|Default|GetTireModWID) 1 (Math|Vector|MakeVector _returnvalue_x _returnvalue_y (+ _returnvalue_z (Variables|Default|GetSpawn%20Pickup%20Z%20Offset))) "0.000000,1.000000,1.000000" 0 true true false -1 "Tossed" "Unset" 0 true 0)
          (Actor|DestroyActor))
        (:Remote
          (Effects|Components|ParticleSystem|SpawnFXAtLocation (Variables|Default|GetPickUpFX) (Transformation|GetActorLocation)))))) ; Must be called inside EVENT and not function to avoid multiple executions
)

(event EventPlayJiggleTiresAnimation
  (params (StartFrame :type Float%20%28double-precision%29) (EndFrame :type Float%20%28double-precision%29))
  ;[0.15 0 0.1492 0.5] Jiggle Animations
  (CallFunction|MIDPlayAnimation StartFrame EndFrame))

(event EventOnDamagePlayEffects
  (params (Damage :type Float%20%28single-precision%29) (DamageTags :type Gameplay%20Tag%20Container%20Structure%20%28by%20ref%29) (Momentum :type Vector :default "0, 0, 0") (HitInfo :type Hit%20Result%20Structure%20%28by%20ref%29) (InstigatedBy :type Fort%20Pawn%20Object%20Reference) (DamageCauser :type Actor%20Object%20Reference) (EffectContext :type Gameplay%20Effect%20Context%20Handle%20Structure))
  ;[0.15 0 0.1492 0.5] Jiggle Animations
  ;;[1 1 1] When we take damage, do a jiggle!
  (CallFunction|MIDPlayAnimation 42.0 54.0))

(event EventSetLifeSpanForMovingPlatformAttachedBuildingActors
  (params (LifeSpan :type Float%20%28single-precision%29))
  (|CallParentFunction (ScalableFloat|GetValueatLevel (Variables|Default|GetAttachmentLifespan))))

(fn BlueprintCanInteract
  (params (InteractingPawn :type Fort%20Pawn%20Object%20Reference) (InteractionBeingAttempted :type EInteractionBeingAttempted%20Enum :default "FirstInteraction") (InteractionType :type TInteractionType%20Enum :default "IT_NoInteraction"))
  (return (not (Variables|Default|GetFuseLit))))

(fn BlueprintGetInteractionString
  (params (InteractingPawn :type Fort%20Pawn%20Object%20Reference) (InteractionBeingAttempted :type EInteractionBeingAttempted%20Enum :default "FirstInteraction"))
  (return "NSLOCTEXT(\"[D0F7C81B8C5E34FFBDDC4B983C3590AC]\", \"4F9D5C1A4FAC595181EB68808484EB4C\", \"Pick up: Lowrider Hyrdaulic Mod\")"))

(fn BlueprintGetInteractionTime
  (params (InteractingPawn :type Fort%20Pawn%20Object%20Reference) (InteractionBeingAttempted :type EInteractionBeingAttempted%20Enum :default "FirstInteraction"))
  (return 0.25 false))

(fn BlueprintOnLocalInteract
  (params (InteractingPawn :type Fort%20Player%20Pawn%20Object%20Reference) (InteractionBeingAttempted :type EInteractionBeingAttempted%20Enum :default "FirstInteraction"))
  ;[1 1 1] Returns whether the player's inventory is full (while the player is holding pickaxe so the pickup won't replace another quickbar slotted item)
  (if (not (> (Inventory|CountInventoryOverflowfromAddingItem (Fortnite|GetFortPlayerControllerfromActor InteractingPawn) (Loot|CreateItemEntry (Variables|Default|GetWeaponItemDefinition)) false true) 0))
    (return false)
    (else
      ;;[0.1313 0 0.15] What happens when the player's inventory IS full and they try to pick up this mod
      (FailedAction|UnabletoPerformAction InteractingPawn true (Variables|Default|GetTag_FailedInteract_InventoryIsFull))
      (Audio|PlaySoundatLocation (Variables|Default|GetFailedInteractSound) (Transformation|GetActorLocation))
      (return true))))

(fn CalcVehicleImpulse
  (params (Vehicle :type Actor%20Object%20Reference) (VelocityMultiplier :type Float%20%28double-precision%29))
  (return (Math|Vector|ClampVectorSize (Math|Vector|MakeVector (* (.x (Transformation|GetVelocity Vehicle)) VelocityMultiplier) (* (.y (Transformation|GetVelocity Vehicle)) VelocityMultiplier) (Math|Float|Clamp(Float) (* VelocityMultiplier (* (Math|Vector|VectorLength (Transformation|GetVelocity Vehicle)) 0.5)) (Variables|Default|GetMinVehicleImpulseHeight) (Variables|Default|GetMaxVehicleImpulseHeight))) (Variables|Default|GetMinTotalVehicleImpulse) (Variables|Default|GetMaxTotalVehicleImpulse))))

(fn CeilingCheck
  (bind _overlappawn (Variables|Default|GetOverlapPawn)
    (:then
      (Utilities|Array|AddUnique (LocalVariables|GetActorsToIgnore) _overlappawn)
      (Utilities|Array|AppendArray (LocalVariables|GetActorsToIgnore) (Actor|GetAttachedActors _overlappawn true false))
      (for _array_element (Actor|GetAttachedActors _overlappawn true false)
        (Utilities|Array|AppendArray (LocalVariables|GetActorsToIgnore) (Actor|GetAttachedActors _array_element true false))))
    (:else
      _))
  (bind (_outhit _returnvalue) (Collision|CapsuleTraceByChannel (Math|Vector|MakeVector (.x (Transformation|GetWorldLocation (Variables|Transformation|GetRootComponent))) (.y (Transformation|GetWorldLocation (Variables|Transformation|GetRootComponent))) (+ (.z (Transformation|GetWorldLocation (Variables|Transformation|GetRootComponent))) 5e+01)) (Math|Vector|MakeVector (.x (Transformation|GetWorldLocation (Variables|Transformation|GetRootComponent))) (.y (Transformation|GetWorldLocation (Variables|Transformation|GetRootComponent))) (+ (.z (Transformation|GetWorldLocation (Variables|Transformation|GetRootComponent))) 4.5e+02)) 50 100 "TraceTypeQuery9" false (LocalVariables|GetActorsToIgnore)))
  (return _returnvalue))

(fn DebugJumpCapsule
  (bind _jumpcollision (Variables|Default|GetJumpCollision))
  (bind _returnvalue (Transformation|GetWorldLocation _jumpcollision))
  (bind _returnvalue_2 (Transformation|GetWorldRotation _jumpcollision))
  (bind _capsuleradius (Variables|/Script/Engine.CapsuleComponent|GetCapsuleRadius _jumpcollision))
  (bind _capsulehalfheight (Variables|/Script/Engine.CapsuleComponent|GetCapsuleHalfHeight _jumpcollision))
  (Networking|SwitchHasAuthority
    (:Authority
      (Rendering|Debug|DrawDebugCapsule _returnvalue _capsulehalfheight _capsuleradius _returnvalue_2 "(R=0.000000,G=0.041747,B=1.000000,A=1.000000)" 8.0 1.0) ; server
    )
    (:Remote
      (Rendering|Debug|DrawDebugCapsule _returnvalue _capsulehalfheight _capsuleradius _returnvalue_2 "(R=0.654514,G=0.646449,B=0.000000,A=1.000000)" 8.0 1.0) ; client
    )))

(fn DetermineBounceHeights
  (bind _success (CallFunction|CeilingCheck))
  (Variables|Default|SetBounceHeight (select _success 6e+02 2e+03))
  (Variables|Default|SetJumpHeight (select _success 1e+02 6e+02))
  (Variables|Default|SetFinalBounceJumpHeight (Variables|Default|GetBounceHeight)))

(fn IsValetVehicle
  (params (Actor :type Actor%20Object%20Reference))
  (bind _asfort_vehicle_interface (Utilities|Casting|CastToFortVehicleInterface Actor)
    (:then
      (if (GameplayTags|HasMatchingGameplayTag (Vehicle|GetVehicleAbilitySystemComponent _asfort_vehicle_interface) "(TagName=\"Vehicle.Valet\")")
        (return true _asfort_vehicle_interface)
        (else
          _)))
    (:CastFailed
      _))
  (return false 0))

(fn MID_Play_Animation
  (params (StartFrame :type Float%20%28double-precision%29) (EndFrame :type Float%20%28double-precision%29))
  (bind _tiremid (Variables|Default|GetTireMID))
  (bind _tire (Variables|Default|GetTire))
  (Utilities|IsValid (Variables|Default|GetTireMID)
    (:Is%20Valid
      _)
    (:Is%20Not%20Valid
      (bind _returnvalue (Rendering|Material|CreateDynamicMaterialInstance (Variables|Default|GetTire) 0 (Rendering|Material|GetMaterial (Variables|Default|GetTire) 0) "None" :!class /Script/Engine.PrimitiveComponent))
      (Variables|Default|SetTireMID _returnvalue)))
  (Rendering|Material|SetScalarParameterValue _tiremid "GameTime" (Utilities|Time|GetGameTimeinSeconds) :!class /Script/Engine.MaterialInstanceDynamic)
  (Rendering|Material|SetScalarParameterValue _tiremid "StartFrame" StartFrame :!class /Script/Engine.MaterialInstanceDynamic)
  (Rendering|Material|SetScalarParameterValue _tiremid "EndFrame" EndFrame :!class /Script/Engine.MaterialInstanceDynamic)
  ;[1 1 1] Used with custom primitive paramaters
  (Rendering|Material|SetCustomPrimitiveDataFloat _tire 0 (Utilities|Time|GetGameTimeinSeconds)) ; Game Time
  (Rendering|Material|SetCustomPrimitiveDataFloat _tire 1 5.0) ; Start Frame
  (Rendering|Material|SetCustomPrimitiveDataFloat _tire 2 22.0) ; End Frame
)

(fn OnRep_JumpCollisionLocation
  (bind _jumpcollisionlocation (Variables|Default|GetJumpCollisionLocation))
  (bind _jumpcollision (Variables|Default|GetJumpCollision))
  (if (Utilities|Operators|NotEqual(!=) _jumpcollisionlocation (Transformation|GetWorldLocation _jumpcollision) 0.0001)
    (Transformation|SetWorldLocation _jumpcollision _jumpcollisionlocation false true)))

(fn OnRep_LastHit
  (bind _self self)
  (Networking|CanExecuteCosmeticEvents
    (:True
      (bind _outcueparameters (GameplayCue|BuildCueParametersfromHitResult _self (Variables|Default|GetLastHit)))
      (Projectile|AddGameplayCue%28Looping%29Local _self (Variables|GameplayCues|GetGC_OnFire) _outcueparameters))))

(fn OnRep_OverlapVehicle
  (CallFunction|LaunchVehicle))

(fn OnRep_ShouldExplode)

(fn ShouldApplyMod
  (params (Vehicle :type Fort%20Vehicle%20Interface%20Interface))
  (return (Math|Boolean|ORBoolean (and (not (Vehicle|Tire|AreAnyTiresPopped Vehicle)) (Vehicle|Mod|Tire|HasTireMod Vehicle)) (Vehicle|IsInoperable Vehicle) (<= (Class|FortVehicleInterface|GetHealth Vehicle) 0.0))))

(fn ShouldDie
  (params (Damage :type Float%20%28single-precision%29) (EventInstigator :type Controller%20Object%20Reference) (DamageCauser :type Actor%20Object%20Reference) (bIsTest :type Boolean) (EffectContext :type Gameplay%20Effect%20Context%20Handle%20Structure%20%28by%20ref%29))
  ;[1 1 1] This is set to false so that we don't lose collision when this is shot out of a player's hand - call Hide and Destroy to finish killing the object
  (return false))

(fn ConstructionScript
  (|CallParentFunction))
