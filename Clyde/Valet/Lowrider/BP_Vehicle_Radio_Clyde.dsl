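; Vehicle radio Blueprint that streams indexed stations through a media player and a speaker component.
; StartRadio requests the station's URL from a FortStreamMediaSource, and PlayRadio opens the source with live-dependent options.
; Capped retries restart at station 0 after URL failures or replay the cached URL after open failures.
; Speaker ticking runs only during playback; closing unbinds delegates, sends streaming analytics and resets state.
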(event EventBeginPlay
  (Components|Tick|SetComponentTickEnabled (Variables|Default|GetRadioSpeaker) false))

(event Collision|EventActorBeginOverlap
  (params (OtherActor :type Actor%20Object%20Reference)))

(event EventTick
  (params (DeltaSeconds :type Float%20%28single-precision%29)))

(event Custom|StartRadio
  (params (StationIndex :type Integer))
  (bind _radiomediasource (Variables|Default|GetRadioMediaSource))
  (bind _output_get (Variables|Default|SetRadioStationIndex StationIndex))
  (Stream|BindEventtoOnVideoUrlSuccess _radiomediasource (delegate Custom|PlayRadio))
  (Stream|BindEventtoOnVideoUrlFailed _radiomediasource (delegate Custom|RadioURLFailed))
  (Variables|/Script/FortniteGame.FortStreamMediaSource|SetStreamID (Utilities|Array|Get(aref) (Variables|Default|GetRadioStations) _output_get) _radiomediasource)
  (Stream|RequestVideoUrl _radiomediasource (Fortnite|GetFirstLocalFortPlayerController true)))

(event Custom|RadioURLFailed
  (params (RetrievedUrl :type String) (Code :type Integer) (RequestType :type EEpicMediaRequestType%20Enum :default "UNKNOWN"))
  (Variables|Default|SetRadioURLFailure true)
  (CallFunction|RadioRetry))

(event Custom|PlayRadio
  (params (RetrievedUrl :type String) (Code :type Integer) (RequestType :type EEpicMediaRequestType%20Enum :default "UNKNOWN"))
  (bind _outputdelegate (EventDispatchers|CreateEvent :!event RadioClosed))
  (bind _radiomediaplayer (Variables|Default|GetRadioMediaPlayer))
  (Variables|Default|SetRadioURL RetrievedUrl)
  (Media|MediaPlayer|BindEventtoOnMediaOpened _radiomediaplayer (EventDispatchers|CreateEvent :!event RadioPlay))
  (Media|MediaPlayer|BindEventtoOnMediaClosed _radiomediaplayer _outputdelegate)
  (Media|MediaPlayer|BindEventtoOnEndReached _radiomediaplayer _outputdelegate)
  (Media|MediaPlayer|BindEventtoOnMediaOpenFailed _radiomediaplayer (delegate Custom|RadioOpenFailed))
  (Media|MediaPlayer|OpenSourcewithOptions _radiomediaplayer (Variables|Default|GetRadioMediaSource) (select (Variables|/Script/FortniteGame.FortStreamMediaSource|GetbIsLive (Variables|Default|GetRadioMediaSource)) (Utilities|Struct|MakeMediaPlayerOptions "(Audio=0,Caption=-1,Metadata=-1,Script=-1,Subtitle=-1,Text=-1,Video=0)" "(Video=\"\",Audio=\"\",Subtitle=\"\",Caption=\"\")" "UseTrackOptionIndices" (Math|Timespan|FromSeconds 0.0)) (Utilities|Struct|MakeMediaPlayerOptions "(Audio=0,Caption=-1,Metadata=-1,Script=-1,Subtitle=-1,Text=-1,Video=0)")) :!class /Script/MediaAssets.MediaPlayer))

(event Custom|RadioOpenFailed
  (params (FailedUrl :type String))
  (Variables|Default|SetRadioURLFailure false)
  (CallFunction|RadioRetry))

(event Custom|StopRadio
  (Media|MediaPlayer|Close (Variables|Default|GetRadioMediaPlayer) :!class /Script/MediaAssets.MediaPlayer))

(fn RadioClosed
  (CallFunction|RadioUnbindEvents false)
  (CallFunction|RadioSendAnalytics)
  (Components|Tick|SetComponentTickEnabled (Variables|Default|GetRadioSpeaker) false)
  (Variables|Default|SetRadioRetryCounter 0)
  (Variables|Default|SetRadioURL "")
  (Variables|Default|SetRadioURLFailure false)
  (Variables|Default|SetRadioStationIndex 0))

(fn RadioPlay
  (params (URL :type String))
  (Components|Tick|SetComponentTickEnabled (Variables|Default|GetRadioSpeaker) true)
  (Media|MediaPlayer|Play (Variables|Default|GetRadioMediaPlayer) :!class /Script/MediaAssets.MediaPlayer)
  (CallFunction|RadioUnbindEvents true))

(fn RadioRetry
  (bind _radioretrycounter (Variables|Default|GetRadioRetryCounter))
  (Math|Integer|IncrementInt _radioretrycounter
    (:%20%20
      (if (<= _radioretrycounter (Variables|Default|GetRadioRetries))
        (CallFunction|RadioUnbindEvents false)
        (if (Variables|Default|GetRadioURLFailure)
          (CallFunction|StartRadio 0)
          (else
            (CallFunction|PlayRadio (Variables|Default|GetRadioURL) 200 "BLURL")))
        (else
          (CallFunction|RadioClosed)
          (Variables|Default|SetRadioRetryCounter 0))))))

(fn RadioSendAnalytics
  (for _array_element (Fortnite|GetLocalFortPlayerControllers true)
    (Movie|SendStreamingAnalytics _array_element (Variables|Default|GetRadioMediaPlayer))))

(fn RadioUnbindEvents
  (params (JustURL :type Boolean))
  (bind _radiomediasource (Variables|Default|GetRadioMediaSource))
  (bind _radiomediaplayer (Variables|Default|GetRadioMediaPlayer))
  (Stream|UnbindallEventsfromOnVideoUrlSuccess _radiomediasource)
  (Stream|UnbindallEventsfromOnVideoUrlFailed _radiomediasource)
  (if (not JustURL)
    (Media|MediaPlayer|UnbindallEventsfromOnMediaOpened _radiomediaplayer)
    (Media|MediaPlayer|UnbindallEventsfromOnMediaClosed _radiomediaplayer)
    (Media|MediaPlayer|UnbindallEventsfromOnMediaOpenFailed _radiomediaplayer)
    (Media|MediaPlayer|UnbindallEventsfromOnEndReached _radiomediaplayer)))

(fn ConstructionScript)
