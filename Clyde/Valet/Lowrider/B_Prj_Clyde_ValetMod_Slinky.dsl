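; Thrown tire vehicle-mod projectile that, on server authority, applies the mod to eligible vehicles it hits, then destroys itself.
; Other hits launch player pawns or impulse physics objects, with parent collision logic handling spawning.
; Entering water, hitting a cow catcher, or quickly re-hitting a vehicle that refuses the mod (timer-reset, hotfix-gated) turns it back into a world pickup.
; Hit and EnteredWater flags keep collision and water handling from both firing.
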(event Collision|EventHit
  (params (MyComp :type Primitive%20Component%20Object%20Reference) (Other :type Actor%20Object%20Reference) (OtherComp :type Primitive%20Component%20Object%20Reference) (bSelfMoved :type Boolean) (HitLocation :type Vector :default "0, 0, 0") (HitNormal :type Vector :default "0, 0, 0") (NormalImpulse :type Vector :default "0, 0, 0") (Hit :type Hit%20Result%20Structure%20%28by%20ref%29))
  (bind (_bblockinghit _binitialoverlap _time _distance _location _impactpoint _normal _impactnormal _physmat _hitactor) (Collision|BreakHitResult Hit))
  ;[0.695 1 0] EVENT -> On Hit
  ;;[1 1 1] If already entered water, don't do anything with this hit. Set hit so it can be checked in water interaction.
  (if (Variables|Default|GetEnteredWater)
    (Variables|Spawning|ValidityChecks|SetbSpawnOnCollision false)
    (else
      (Variables|Default|SetHit true)
      (bind _output_get (Variables|Default|SetHit%20Actor _hitactor))
      (bind _asfort_athena_vehicle (Utilities|Casting|CastToFortAthenaVehicle _output_get)
        (:then
          ;;[0.4267 0 0.6094] Check to apply mod to vehicle
          (if (Vehicle|Mod|CanApplyMod _asfort_athena_vehicle (Variables|Default|GetModTagToApply))
            (if (CallFunction|ShouldApplyMod _asfort_athena_vehicle)
              (CallFunction|ApplyModtoVehicle _asfort_athena_vehicle)
              (else
                ;[0 0.5198 1] If we hit the same vehicle multiple times very quickly then spawn a pickup. This is to fix FORT-364546
                (if (> (ScalableFloat|GetValueatLevel (Variables|Default|GetHotfixSpawnPickUpIfStuckOnVehicle)) 0.0)
                  (Networking|SwitchHasAuthority
                    (:Authority
                      (if (Utilities|Equal(Object) (Variables|Default|GetHitVehicle) _asfort_athena_vehicle)
                        (CallFunction|SpawnPickupMod)
                        (else
                          (Variables|Default|SetHitVehicle _asfort_athena_vehicle)
                          (Utilities|Time|SetTimerbyEvent (delegate Custom|ResetHitVehicle) 0.2))))))))
            (else
              (CallFunction|OnHitNonVehicle Hit))))
        (:CastFailed
          (bind _hit_actor (Variables|Default|GetHit%20Actor)
            (:then
              (bind _returnvalue (GameplayTags|HasMatchingGameplayTag(Message) _hit_actor "(TagName=\"Vehicle.Mod.CowCatcher\")")) ; FORT-469775- If we hit a cow catcher just spawn the pickup
              (if _returnvalue
                (CallFunction|SpawnPickupMod)
                (else
                  (CallFunction|OnHitNonVehicle Hit))))
            (:else)))))))

(event Custom|ParentHit
  ;[0.695 1 0] EVENT -> On Hit
  (|CallParentFunction (stale Collision|EventHit MyComp) (stale Collision|EventHit Other) (stale Collision|EventHit OtherComp) (stale Collision|EventHit bSelfMoved) (stale Collision|EventHit HitLocation) (stale Collision|EventHit HitNormal) (stale Collision|EventHit NormalImpulse) (stale Collision|EventHit Hit))
  (if (Variables|Spawning|GetSpawned)
    (Collision|SetActorEnableCollision false) ; If we spawned the BGA then disable collision so spawned BGA doesn't collide with us
  ))

(event Custom|ResetHitVehicle
  ;[0 0.5198 1] If we hit the same vehicle multiple times very quickly then spawn a pickup. This is to fix FORT-364546
  (Variables|Default|SetHitVehicle 0))

(event Custom|onHitNonVehicle
  (params (Hit :type Hit%20Result%20Structure%20%28by%20ref%29))
  (bind _hit_actor (Variables|Default|GetHit%20Actor))
  ;[0.9993 1 0] Handle hitting non-vehicle actors
  (bind _hit_actor_2 (Variables|Default|GetHit%20Actor)
    (:then
      (bind _asfort_pawn (Utilities|Casting|CastToFortPawn _hit_actor_2)
        (:then
          (bind _asfort_player_pawn (Utilities|Casting|CastToFortPlayerPawn _hit_actor)
            (:then
              (if (or (Vehicle|IsinVehicle _asfort_player_pawn) (Health|IsDBNO _asfort_player_pawn))
                _
                (elif (ScalableFloat|GetValueasBool (Variables|HotfixRows|GetShouldLaunchHitPlayer))
                  (Utilities|LaunchPawns (Actor|GetOwner) _asfort_pawn (CallFunction|CalculateLaunchVelocity _asfort_pawn) true true false true true true true)
                  (Projectile|SetGravityScale 8.0) ; Don't bounce as far after hitting a pawn
                  (else
                    (CallFunction|ParentHit)))))
            (:CastFailed
              _)))
        (:CastFailed
          (Utilities|Casting|CastToB_Prj_Clyde_ValetMod_Slinky _hit_actor
            (:then
              (Development|PrintString "HITTING SELF"))
            (:CastFailed
              ;;[0 1 0.2298] Should the projectile bounce off tire piles or do we allow players to stack one on the other
              (if (not (ScalableFloat|GetValueasBool (Variables|HotfixRows|GetShouldTirePilesBeStackable)))
                (bind _hit_actor_3 (Variables|Default|GetHit%20Actor)
                  (:then
                    (Utilities|Casting|CastToBGA_Clyde_ValetMod_Slinky_Pickup _hit_actor_3
                      (:CastFailed
                        _)))
                  (:else)))
              (CallFunction|ImpactLaunch Hit))))))
    (:else)))

(event Custom|ImpactLaunch
  (params (Hit :type Hit%20Result%20Structure%20%28by%20ref%29))
  (bind (_bblockinghit _binitialoverlap _time _distance _location _impactpoint _normal _impactnormal _physmat _hitactor) (Collision|BreakHitResult Hit))
  ;[1 0 0.1108] Physics Object -> Launch On Impact -- disabled
  (bind _returnvalue (FortPhysicsSimulationLibrary|CanActorSimulatePhysics _hitactor))
  (if (and _returnvalue (Utilities|IsValid (Actor|GetComponentbyClass _hitactor "/Script/FortniteGame.FortPhysicsObjectComponent") :!class /Script/Engine.KismetSystemLibrary))
    (bind _returnvalue_2 (FortPhysicsSimulationLibrary|GetImpulseVectorforDirectHit _impactnormal (ScalableFloat|GetValueatLevel (Variables|Default|GetPhysicsObjectImpulseStrength))))
    (FortPhysicsSimulationLibrary|AddPhysicsImpulseatLocation _hitactor _returnvalue_2 _impactpoint true "None" false 0 "(MaxSimSizeAllowed=Large,bRejectEmbeddedInLandscape=False)")
    (else
      _))
  (CallFunction|ParentHit)
  (Variables|Default|SetHit false) ; Need this in case we bounce into water (FORT-380749)
)

(event EventOnStop
  (params (Hit :type Hit%20Result%20Structure%20%28by%20ref%29))
  ;[1 1 1] EVENT -> On Stop
  (|CallParentFunction Hit)
  (Rendering|SetVisibility (Variables|Default|GetTire) false true))

(event EventBeginPlay
  ;[1 1 1] BeginPlay -> Fix throwing the projectile whilst in water, to allow it time to leave the pawns hands
  (|CallParentFunction)
  (Utilities|FlowControl|Delay 0.2)
  (if (Water|IsinWater (Variables|Water|GetWaterInteractionComponent))
    (CallFunction|HasEnteredWater)))

(event Custom|HasEnteredWater
  (bind _capsulecomponent (Variables|Collision|GetCapsuleComponent))
  (bind _returnvalue (Transformation|GetWorldLocation _capsulecomponent))
  ;[1 1 1] Has Entered Water - Spawn Pickup
  ;;[1 1 1] If already entered Hit, don't do anything with the water interaction. Set Entered Water so it can be checked in Hit interaction.
  (if (not (Variables|Default|GetHit))
    (Utilities|FlowControl|DoOnce false
      (:Completed
        (Variables|Default|SetEnteredWater true)
        (Effects|Components|ParticleSystem|SpawnFXAtLocation (Variables|Default|GetWaterSplashFX) _returnvalue)
        (Rendering|SetActorHiddenInGame true)
        (Inventory|SpawnPickupInWorld (Variables|Default|GetTireModWID) 1 _returnvalue "0.000000,0.000000,1.000000" 0 true false)
        (Actor|DestroyActor)))))

(event OnEnterWaterEvent(WaterInteractionComponent)
  (params (WaterBody :type Fort%20Water%20Body%20Actor%20Object%20Reference) (WaterInteractionComponent :type Fort%20Water%20Interaction%20Component%20Object%20Reference) (bIsFirstBody :type Boolean))
  ;[1 1 1] Has Entered Water - Spawn Pickup
  (Variables|Spawning|ValidityChecks|SetbSpawnOnCollision false)
  (CallFunction|HasEnteredWater))

(event Custom|Apply%20Mod%20To%20Vehicle
  (params (Vehicle :type Fort%20Athena%20Vehicle%20Object%20Reference))
  ;[0 0.1767 0.4913] Apply the mod to the vehicle
  (Utilities|FlowControl|DoOnce false
    (:Completed
      (Networking|SwitchHasAuthority
        (:Authority
          (bind _returnvalue (PlayerInfo|GetOwnerPlayerController self))
          (Vehicle|Mod|ApplyVehicleMod Vehicle (Variables|Default|GetModTagToApply) _returnvalue)
          (Rendering|SetVisibility (Variables|Default|GetTire) false true)
          (Actor|DestroyActor))))))

(event Custom|SpawnPickupMod
  ;[0.15 0 0.145] Spawn Pickup Mod
  (Utilities|FlowControl|DoOnce false
    (:Completed
      (Rendering|SetActorHiddenInGame true)
      (Inventory|SpawnPickupInWorld (Variables|Default|GetTireModWID) 1 (Transformation|GetWorldLocation (Variables|Collision|GetCapsuleComponent)) "0.000000,0.000000,1.000000" 0 true false)
      (Actor|DestroyActor))))

(fn CalculateLaunchVelocity
  (params (PlayerPawn :type Fort%20Pawn%20Object%20Reference))
  (return (Math|Vector|MakeVector (.x (* (Math|Vector|Normalize (- (Transformation|GetActorLocation PlayerPawn) (Transformation|GetActorLocation))) (ScalableFloat|GetValueatLevel (Variables|HotfixRows|GetHitPlayerLaunchMag)))) (.y (* (Math|Vector|Normalize (- (Transformation|GetActorLocation PlayerPawn) (Transformation|GetActorLocation))) (ScalableFloat|GetValueatLevel (Variables|HotfixRows|GetHitPlayerLaunchMag)))) (* (ScalableFloat|GetValueatLevel (Variables|HotfixRows|GetHitPlayerLaunchMag)) (ScalableFloat|GetValueatLevel (Variables|HotfixRows|GetHitPlayerLaunchZMulti))))))

(fn ShouldApplyMod
  (params (Vehicle :type Fort%20Athena%20Vehicle%20Object%20Reference))
  ;[1 1 1] If none of these are true then we should apply the mod
  (return (not (Math|Boolean|ORBoolean (and (not (Vehicle|Tire|AreAnyTiresPopped Vehicle)) (Vehicle|Mod|Tire|HasTireMod Vehicle)) (Vehicle|IsInoperable Vehicle) (<= (Class|FortVehicleInterface|GetHealth Vehicle) 0.0)))))

(fn ConstructionScript
  (|CallParentFunction))
