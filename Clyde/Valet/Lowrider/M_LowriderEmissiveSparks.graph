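# M_LowriderEmissiveSparks is a Default Lit material for Niagara sprite particles, named for glowing sparks on the Lowrider valet asset.
# Its graph multiplies a fixed orange color by ten, producing an HDR value suited to emissive glow.
# It contains no textures, parameters or particle color input, so the color and intensity are hard-coded.
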Material /Clyde/Valet/Lowrider/M_LowriderEmissiveSparks
graph {
 properties {bAutomaticallySetUsageInEditor=false, bUsedWithNiagaraSprites=true, usedShadingModels=Default Lit}
 nodes [Multiply(A=Constant3Vector() {constant=<rgba 1 0.3527 0 1>}) {constB=10}]
}
