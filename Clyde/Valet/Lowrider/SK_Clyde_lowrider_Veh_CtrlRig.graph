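# Control Rig for Clyde's lowrider vehicle skeleton, posing the root, frame, body and four wheel assemblies from controls.
# One execution chain copies global control transforms onto the root, frame and body bones.
# FrontWheels and RearWheels master controls overwrite their left/right controls' local transforms, which then drive the front axle-pivot and rear wheel-steering bones.
# Most controls are parented under the bones they drive; the two masters sit under Root_CTRL.
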ControlRigBlueprint /Clyde/Valet/Lowrider/SK_Clyde_lowrider_Veh_CtrlRig
graphs [
 RigVMFunctionLibrary {links [], nodes []}
 RigVMModel {nodes [
   RigUnit_BeginExecution = &RigUnit_BeginExecution()
   RigUnit_GetTransform = &RigUnit_GetTransform(Item=(Type=Control,Name="Root_CTRL"), bInitial=False)
   RigUnit_SetTransform = &RigUnit_SetTransform(Item=(Type=Bone,Name="root"), Weight=1, bInitial=False, ExecutePin=@RigUnit_BeginExecution.ExecutePin, Value=@RigUnit_GetTransform.Transform)
   > &RigUnit_SetTransform(
    Item=(Type=Bone,Name="frame")
    Weight=1
    bInitial=False
    Value=&RigUnit_GetTransform(Item=(Type=Control,Name="frame_ctrl"), bInitial=False).Transform
   )
   > &RigUnit_SetTransform(
    Item=(Type=Bone,Name="body")
    Weight=1
    bInitial=False
    Value=&RigUnit_GetTransform(Item=(Type=Control,Name="body_ctrl"), bInitial=False).Transform
   )
   n1 = &RigUnit_GetTransform(Item=(Type=Control,Name="FrontWheels_CTRL"), Space=LocalSpace, bInitial=False)
   > /Script/ControlRig.RigUnit_SetTransformItemArray(
    Items=((Type=Control,Name="FrontLeft_CTRL"),(Type=Control,Name="FrontRight_CTRL"))
    Space=LocalSpace
    Transforms=((Rotation=(X=0,Y=0,Z=0,W=1),Translation=(X=0,Y=0,Z=0),Scale3D=(X=1,Y=1,Z=1)),(Rotation=(X=0,Y=0,Z=0,W=1),Translation=(X=0,Y=0,Z=0),Scale3D=(X=1,Y=1,Z=1)))
    Weight=1
    bInitial=False
    bPropagateToChildren=True
    Transforms.0=@n1.Transform
    Transforms.1=@n1.Transform
   ) {is=struct RigVMUnitNode}
   n2 = &RigUnit_GetTransform(Item=(Type=Control,Name="RearWheels_CTRL"), Space=LocalSpace, bInitial=False)
   > &RigUnit_SetTransformItemArray(
    Items=((Type=Control,Name="BackLeft_CTRL"),(Type=Control,Name="BackRight_CTRL"))
    Space=LocalSpace
    Transforms=((Rotation=(X=0,Y=0,Z=0,W=1),Translation=(X=0,Y=0,Z=0),Scale3D=(X=1,Y=1,Z=1)),(Rotation=(X=0,Y=0,Z=0,W=1),Translation=(X=0,Y=0,Z=0),Scale3D=(X=1,Y=1,Z=1)))
    Weight=1
    bInitial=False
    bPropagateToChildren=True
    Transforms.0=@n2.Transform
    Transforms.1=@n2.Transform
   )
   > &RigUnit_SetTransform(
    Item=(Type=Bone,Name="wheel_steering_bk_l")
    Weight=1
    bInitial=False
    Value=&RigUnit_GetTransform(Item=(Type=Control,Name="BackLeft_CTRL"), bInitial=False).Transform
   )
   > &RigUnit_SetTransform(
    Item=(Type=Bone,Name="wheel_steering_bk_r")
    Weight=1
    bInitial=False
    Value=&RigUnit_GetTransform(Item=(Type=Control,Name="BackRight_CTRL"), bInitial=False).Transform
   )
   > &RigUnit_SetTransform(
    Item=(Type=Bone,Name="axel_pivot_fr_l")
    Weight=1
    bInitial=False
    Value=&RigUnit_GetTransform(Item=(Type=Control,Name="FrontLeft_CTRL"), bInitial=False).Transform
   )
   > &RigUnit_SetTransform(
    Item=(Type=Bone,Name="axel_pivot_fr_r")
    Weight=1
    bInitial=False
    Value=&RigUnit_GetTransform(Item=(Type=Control,Name="FrontRight_CTRL"), bInitial=False).Transform
   )
  ]}
]
hierarchy [
 root {loc=<0 0 0>, rot=<0 0 0 1>}
 frame {parent=root, loc=<-0.0727 0 14.3516>, rot=<0 0 0 1>}
 body {parent=frame, loc=<0 0 27.8584>, rot=<0 0 0 1>}
 steering_column {parent=body, loc=<105.808 -34.9436 32.0624>, rot=<0 -0.0874 0 -0.9962>}
 steering_wheel {parent=steering_column, loc=<-48.2677 0 -0.2251>, rot=<0 0 0 1>}
 mirror_l {parent=body, loc=<91.6992 -92 47.3567>, rot=<0 0 0 1>}
 glass_c {parent=body, loc=<95.6457 0 50.1049>, rot=<0 0 0 1>}
 glass_l {parent=body, loc=<6.3007 -75 49.5744>, rot=<0 0 0 1>}
 glass_r {parent=body, loc=<6.3007 75 49.5744>, rot=<0 -1 0 0>}
 mirror_r {parent=body, loc=<91.6991 92 47.3566>, rot=<0 -1 0 0>}
 antenna {parent=body, loc=<21.3354 0 86.6808>, rot=<0.1512 0 0.9885 0>}
 spoiler_break {parent=body, loc=<-211.0808 -63.0626 64.3169>, rot=<0 0 1 0>}
 wheel_well_spreader_fr_l {parent=body, loc=<156.1915 -110.3726 -13.4052>, rot=<0 0 0 1>}
 wheel_well_spreader_bk_l {parent=body, loc=<-100.3321 -128.0818 -8.8054>, rot=<0 0 0 1>}
 wheel_well_spreader_fr_r {parent=body, loc=<156.1911 110.3722 -13.4052>, rot=<0 0 0 1>}
 wheel_well_spreader_bk_r {parent=body, loc=<-100.3317 128.0816 -8.8054>, rot=<0 0 0 1>}
 Body_CTRL {controlType=EULER_TRANSFORM, parent=body, type=Control, loc=<0 0 0>, rot=<0 0 0 1>}
 rear_differential {parent=frame, loc=<-102.0366 0 9.0432>, rot=<0 0 0 1>}
 wheel_steering_bk_l {parent=rear_differential, loc=<1.71 -77.7477 6.2419>, rot=<0 0 0 1>}
 wheel_tilted_bk_l {parent=wheel_steering_bk_l, loc=<0 -44.8619 3.9723>, rot=<0.0441 0.0001 -0.0017 -0.999>}
 tire_bk_l {parent=wheel_tilted_bk_l, loc=<-0.0011 11.544 0>, rot=<0 0 0 1>}
 BackLeft_CTRL {controlType=EULER_TRANSFORM, parent=wheel_steering_bk_l, type=Control, loc=<0 0 0>, rot=<0 0 0 1>}
 wheel_steering_bk_r {parent=rear_differential, loc=<1.7097 77.7477 6.2419>, rot=<0 0 0 1>}
 wheel_tilted_bk_r {parent=wheel_steering_bk_r, loc=<0 44.8621 3.9723>, rot=<-0.0441 -0.0001 -0.0017 -0.999>}
 tire_bk_r {parent=wheel_tilted_bk_r, loc=<0.0011 -11.544 0>, rot=<0 0 0 1>}
 BackRight_CTRL {controlType=EULER_TRANSFORM, parent=wheel_steering_bk_r, type=Control, loc=<0 0 0>, rot=<0 0 0 1>}
 front_differential {parent=frame, loc=<156.1539 0 10.2679>, rot=<0 0 0 1>}
 axel_pivot_fr_l {parent=front_differential, loc=<0 -27.9114 0>, rot=<0 0 0 1>}
 steering_knuckle_fr_l {parent=axel_pivot_fr_l, loc=<0 -20.2959 1.3542>, rot=<0 0 0 1>}
 wheel_steering_fr_l {parent=steering_knuckle_fr_l, loc=<0 -37.8233 1.0436>, rot=<0 0 0 1>}
 wheel_tilted_fr_l {parent=wheel_steering_fr_l, loc=<0 -18.1253 1.5858>, rot=<0.0436 0 0 -0.999>}
 tire_fr_l {parent=wheel_tilted_fr_l, loc=<0 11.544 0>, rot=<0 0 0 1>}
 FrontLeft_CTRL {controlType=EULER_TRANSFORM, parent=axel_pivot_fr_l, type=Control, loc=<0 0 0>, rot=<0 0 0 1>}
 axel_pivot_fr_r {parent=front_differential, loc=<0 27.2678 0>, rot=<0 0 0 1>}
 steering_knuckle_fr_r {parent=axel_pivot_fr_r, loc=<0 20.2959 1.3542>, rot=<0 0 0 1>}
 wheel_steering_fr_r {parent=steering_knuckle_fr_r, loc=<0 37.8233 1.0436>, rot=<0 0 0 1>}
 wheel_tilted_fr_r {parent=wheel_steering_fr_r, loc=<0 18.1251 1.5858>, rot=<-0.0436 0 0 -0.999>}
 tire_fr_r {parent=wheel_tilted_fr_r, loc=<0 -11.544 0>, rot=<0 0 0 1>}
 FrontRight_CTRL {controlType=EULER_TRANSFORM, parent=axel_pivot_fr_r, type=Control, loc=<0 0 0>, rot=<0 0 0 1>}
 exhaust_l {parent=frame, loc=<-73.7867 -31.3609 7.9816>, rot=<0 0 0 1>}
 exhaust_r {parent=frame, loc=<-73.7867 31.3609 7.9816>, rot=<0 -1 0 0>}
 Frame_CTRL {controlType=EULER_TRANSFORM, parent=frame, type=Control, loc=<0 0 0>, rot=<0 0 0 1>}
 Root_CTRL {controlType=EULER_TRANSFORM, parent=root, type=Control, loc=<0 0 0>, rot=<0 0 0 1>}
 FrontWheels_CTRL {controlType=EULER_TRANSFORM, parent=Root_CTRL, type=Control, loc=<0 0 0>, rot=<0 0 0 1>}
 RearWheels_CTRL {controlType=EULER_TRANSFORM, parent=Root_CTRL, type=Control, loc=<0 0 0>, rot=<0 0 0 1>}
]
