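# Animation Blueprint for Clyde's Valet Lowrider on the Oblong car skeleton, adding hydraulic-style tilts, drops and bounces.
# AnimGraph chains bone transforms from mesh-space ref pose: wheels, axles, steering, tire-mod widening, bouncy chassis, then additive hydraulic offsets with bool-alpha blends.
# Update sets tilt flags from steering angle.
# Vehicle delegates trigger a timed bounce, a left drop followed by a timed right drop, and drop reset.
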AnimBlueprint /Clyde/Valet/Lowrider/SK_VEH_Oblong_AnimBP_Clyde
target_skeleton=/VehicleCosmetics/Bodies/Oblong/Meshes/SK_VEH_Oblong_Skeleton
graphs [
 AnimGraph {nodes [
   n1 = "Transform (Modify) Bone - Bone: front_differential"(
    ComponentPose=Mesh Space Ref Pose!() -> (ComponentPose:Component Space Pose Link Structure) {is=title AnimGraphNode_MeshRefPose}
    Alpha=1
    Translation=Get AxlePositionF() -> (AxlePositionF:Vector)
    Rotation=Get AxleRotationF() -> (AxleRotationF:Rotator)
    Scale="1,1,1"
   ) {
    rotationMode=BMM_Replace
    rotationSpace=BCS_BoneSpace
    translationMode=BMM_Replace
    translationSpace=BCS_BoneSpace
    alphaBoolBlend {blendOption=Linear}
    boneToModify {boneName=front_differential}
   }
   n2 = "Transform (Modify) Bone - Bone: rear_differential"(
    ComponentPose=@n1
    Alpha=1
    Translation=Get AxlePositionB() -> (AxlePositionB:Vector)
    Rotation=Get AxleRotationB() -> (AxleRotationB:Rotator)
    Scale="1,1,1"
   ) {
    rotationMode=BMM_Replace
    rotationSpace=BCS_BoneSpace
    translationMode=BMM_Replace
    translationSpace=BCS_BoneSpace
    alphaBoolBlend {blendOption=Linear}
    boneToModify {boneName=rear_differential}
   }
   n3 = "Transform (Modify) Bone - Bone: axle_pivot_fr_l"(ComponentPose=@n2, Alpha=1, Translation=Get WheelPositionFL() -> (WheelPositionFL:Vector), Rotation="0,0,0", Scale="1,1,1") {
    rotationMode=BMM_Replace
    rotationSpace=BCS_BoneSpace
    translationMode=BMM_Replace
    translationSpace=BCS_BoneSpace
    alphaBoolBlend {blendOption=Linear}
    boneToModify {boneName=axle_pivot_fr_l}
   }
   n4 = "Transform (Modify) Bone - Bone: axle_pivot_fr_r"(ComponentPose=@n3, Alpha=1, Translation=Get WheelPositionFR() -> (WheelPositionFR:Vector), Rotation="0,0,0", Scale="1,1,1") {
    rotationMode=BMM_Replace
    rotationSpace=BCS_BoneSpace
    translationMode=BMM_Replace
    translationSpace=BCS_BoneSpace
    alphaBoolBlend {blendOption=Linear}
    boneToModify {boneName=axle_pivot_fr_r}
   }
   n5 = "Transform (Modify) Bone - Bone: axle_pivot_bk_r"(ComponentPose=@n4, Alpha=1, Translation=Get WheelPositionBR() -> (WheelPositionBR:Vector), Rotation="0,0,0", Scale="1,1,1") {
    rotationMode=BMM_Replace
    rotationSpace=BCS_BoneSpace
    translationMode=BMM_Replace
    translationSpace=BCS_BoneSpace
    alphaBoolBlend {blendOption=Linear}
    boneToModify {boneName=axle_pivot_bk_r}
   }
   n6 = "Transform (Modify) Bone - Bone: axle_pivot_bk_l"(ComponentPose=@n5, Alpha=1, Translation=Get WheelPositionBL() -> (WheelPositionBL:Vector), Rotation="0,0,0", Scale="1,1,1") {
    rotationMode=BMM_Replace
    rotationSpace=BCS_BoneSpace
    translationMode=BMM_Replace
    translationSpace=BCS_BoneSpace
    alphaBoolBlend {blendOption=Linear}
    boneToModify {boneName=axle_pivot_bk_l}
   }
   n7 = "Transform (Modify) Bone - Bone: wheel_steering_fr_r"(ComponentPose=@n6, Alpha=1, Translation="0,0,0", Rotation=Get WheelRotationFR() -> (WheelRotationFR:Rotator), Scale="1,1,1") {
    rotationMode=BMM_Replace
    rotationSpace=BCS_BoneSpace
    translationSpace=BCS_BoneSpace
    alphaBoolBlend {blendOption=Linear}
    boneToModify {boneName=wheel_steering_fr_r}
   }
   n8 = "Transform (Modify) Bone - Bone: wheel_steering_fr_l"(ComponentPose=@n7, Alpha=1, Translation="0,0,0", Rotation=Get WheelRotationFL() -> (WheelRotationFL:Rotator), Scale="1,1,1") {
    rotationMode=BMM_Replace
    rotationSpace=BCS_BoneSpace
    translationSpace=BCS_BoneSpace
    alphaBoolBlend {blendOption=Linear}
    boneToModify {boneName=wheel_steering_fr_l}
   }
   n9 = "Transform (Modify) Bone - Bone: wheel_disc_fr_r"(ComponentPose=@n8, Alpha=1, Translation="0,0,0", Rotation=Get WheelSpinFR() -> (WheelSpinFR:Rotator), Scale="1,1,1") {rotationMode=BMM_Replace, rotationSpace=BCS_BoneSpace, alphaBoolBlend {blendOption=Linear}, boneToModify {boneName=wheel_disc_fr_r}}
   n10 = "Transform (Modify) Bone - Bone: wheel_disc_fr_l"(ComponentPose=@n9, Alpha=1, Translation="0,0,0", Rotation=Get WheelSpinFL() -> (WheelSpinFL:Rotator), Scale="1,1,1") {rotationMode=BMM_Replace, rotationSpace=BCS_BoneSpace, alphaBoolBlend {blendOption=Linear}, boneToModify {boneName=wheel_disc_fr_l}}
   n11 = "Transform (Modify) Bone - Bone: wheel_disc_bk_r"(ComponentPose=@n10, Alpha=1, Translation="0,0,0", Rotation=Get WheelSpinBR() -> (WheelSpinBR:Rotator), Scale="1,1,1") {rotationMode=BMM_Replace, rotationSpace=BCS_BoneSpace, alphaBoolBlend {blendOption=Linear}, boneToModify {boneName=wheel_disc_bk_r}}
   n12 = "Transform (Modify) Bone - Bone: wheel_disc_bk_l"(ComponentPose=@n11, Alpha=1, Translation="0,0,0", Rotation=Get WheelSpinBL() -> (WheelSpinBL:Rotator), Scale="1,1,1") {rotationMode=BMM_Replace, rotationSpace=BCS_BoneSpace, alphaBoolBlend {blendOption=Linear}, boneToModify {boneName=wheel_disc_bk_l}}
   n13 = "Transform (Modify) Bone - Bone: steering_wheel"(ComponentPose=@n12, Alpha=1, Translation="0,0,0", Rotation=Get SteeringRotation() -> (SteeringRotation:Rotator), Scale="1,1,1") {rotationMode=BMM_Replace, rotationSpace=BCS_BoneSpace, alphaBoolBlend {blendOption=Linear}, boneToModify {boneName=steering_wheel}}
   n14 = Get bHasTireMod() -> (bHasTireMod:bool)
   n15 = "Transform (Modify) Bone - Bone: axle_pivot_bk_l"!(ComponentPose=@n13, bAlphaBoolEnabled:bool=@n14, Translation="0,-20,0", Rotation="0,0,0", Scale="1,1,1") -> (Pose) {
    alphaInputType=Bool
    rotationSpace=BCS_BoneSpace
    scaleSpace=BCS_BoneSpace
    translationMode=BMM_Additive
    translationSpace=BCS_BoneSpace
    alphaBoolBlend {blendOption=Linear}
    boneToModify {boneName=axle_pivot_bk_l}
    translation {y=-20}
   }
   n16 = "Transform (Modify) Bone - Bone: axle_pivot_bk_r"!(ComponentPose=@n15, bAlphaBoolEnabled:bool=@n14, Translation="0,20,0", Rotation="0,0,0", Scale="1,1,1") -> (Pose) {
    alphaInputType=Bool
    rotationSpace=BCS_BoneSpace
    scaleSpace=BCS_BoneSpace
    translationMode=BMM_Additive
    translationSpace=BCS_BoneSpace
    alphaBoolBlend {blendOption=Linear}
    boneToModify {boneName=axle_pivot_bk_r}
    translation {y=20}
   }
   n17 = "Transform (Modify) Bone - Bone: axle_pivot_fr_l"!(ComponentPose=@n16, bAlphaBoolEnabled:bool=@n14, Translation="0,-18,0", Rotation="0,0,0", Scale="1,1,1") -> (Pose) {
    alphaInputType=Bool
    rotationSpace=BCS_BoneSpace
    translationMode=BMM_Additive
    translationSpace=BCS_BoneSpace
    alphaBoolBlend {blendOption=Linear}
    boneToModify {boneName=axle_pivot_fr_l}
    translation {y=-18}
   }
   n18 = "Transform (Modify) Bone - Bone: axle_pivot_fr_r"!(ComponentPose=@n17, bAlphaBoolEnabled:bool=@n14, Translation="0,18,0", Rotation="0,0,0", Scale="1,1,1") -> (Pose) {
    alphaInputType=Bool
    rotationSpace=BCS_BoneSpace
    translationMode=BMM_Additive
    translationSpace=BCS_BoneSpace
    alphaBoolBlend {blendOption=Linear}
    boneToModify {boneName=axle_pivot_fr_r}
    translation {y=18}
   }
   n19 = "Transform (Modify) Bone - Bone: wheel_well_spreader_bk_l"!(ComponentPose=@n18, bAlphaBoolEnabled:bool=@n14, Translation="3,0,0", Rotation="0,0,0", Scale="1.25,1,1") -> (Pose) {
    alphaInputType=Bool
    rotationSpace=BCS_BoneSpace
    scaleMode=BMM_Additive
    scaleSpace=BCS_BoneSpace
    translationMode=BMM_Additive
    translationSpace=BCS_BoneSpace
    alphaBoolBlend {blendOption=Linear}
    alphaScaleBiasClamp {bInterpResult=true, interpSpeedDecreasing=5, interpSpeedIncreasing=5}
    boneToModify {boneName=wheel_well_spreader_bk_l}
    scale {x=1.25}
    translation {x=3}
   }
   n20 = "Transform (Modify) Bone - Bone: wheel_well_spreader_bk_r"!(ComponentPose=@n19, bAlphaBoolEnabled:bool=@n14, Translation="3,0,0", Rotation="0,0,0", Scale="1.25,1,1") -> (Pose) {
    alphaInputType=Bool
    rotationSpace=BCS_BoneSpace
    scaleMode=BMM_Additive
    scaleSpace=BCS_BoneSpace
    translationMode=BMM_Additive
    translationSpace=BCS_BoneSpace
    alphaBoolBlend {blendOption=Linear}
    alphaScaleBiasClamp {bInterpResult=true, interpSpeedDecreasing=5, interpSpeedIncreasing=5}
    boneToModify {boneName=wheel_well_spreader_bk_r}
    scale {x=1.25}
    translation {x=3}
   }
   n21 = "Transform (Modify) Bone - Bone: wheel_well_spreader_fr_l"!(ComponentPose=@n20, bAlphaBoolEnabled:bool=@n14, Translation="0,0,0", Rotation="0,0,0", Scale="1.25,1,1") -> (Pose) {
    alphaInputType=Bool
    rotationSpace=BCS_BoneSpace
    scaleMode=BMM_Additive
    scaleSpace=BCS_BoneSpace
    translationMode=BMM_Additive
    translationSpace=BCS_BoneSpace
    alphaBoolBlend {blendOption=Linear}
    alphaScaleBiasClamp {bInterpResult=true, interpSpeedDecreasing=5, interpSpeedIncreasing=5}
    boneToModify {boneName=wheel_well_spreader_fr_l}
    scale {x=1.25}
   }
   n22 = "Transform (Modify) Bone - Bone: wheel_well_spreader_fr_r"!(ComponentPose=@n21, bAlphaBoolEnabled:bool=@n14, Translation="0,0,0", Rotation="0,0,0", Scale="1.25,1,1") -> (Pose) {
    alphaInputType=Bool
    rotationSpace=BCS_BoneSpace
    scaleMode=BMM_Additive
    scaleSpace=BCS_BoneSpace
    translationMode=BMM_Additive
    translationSpace=BCS_BoneSpace
    alphaBoolBlend {blendOption=Linear}
    alphaScaleBiasClamp {bInterpResult=true, interpSpeedDecreasing=5, interpSpeedIncreasing=5}
    boneToModify {boneName=wheel_well_spreader_fr_r}
    scale {x=1.25}
   }
   AnimGraphNode_BouncyChassisVehicleController_1 = Vehicle Bouncy Chassis Controller!(
    ComponentPose:Component Space Pose Link Structure=@n22
    Alpha:float=1
    BouncyChassis:Bouncy Chassis Vehicle State Structure=(PitchDegrees=0,RollDegrees=0,TranslationZ=0)
   ) -> (Pose:Component Space Pose Link Structure) {is=title AnimGraphNode_VehicleBouncyChassisController, alphaBoolBlend {blendOption=Linear}}
   n23 = Get IsTiltRight() -> (IsTiltRight:bool)
   n24 = "Transform (Modify) Bone - Bone: axle_pivot_bk_l"(
    ComponentPose=@AnimGraphNode_BouncyChassisVehicleController_1
    bAlphaBoolEnabled=@n23
    Translation="0,0,-40"
    Rotation="0,0,0"
    Scale="0,0,0"
   ) {
    alphaInputType=Bool
    rotationSpace=BCS_BoneSpace
    translationMode=BMM_Additive
    alphaBoolBlend {blendInTime=0.25, blendOption=Linear, blendOutTime=0.25}
    alphaScaleBiasClamp {bInterpResult=true, interpSpeedDecreasing=5, interpSpeedIncreasing=5}
    boneToModify {boneName=axle_pivot_bk_l}
    scale=<0 0 0>
    translation {z=-40}
   }
   n25 = "Transform (Modify) Bone - Bone: axle_pivot_fr_l"(ComponentPose=@n24, bAlphaBoolEnabled=@n23, Translation="0,0,-40", Rotation="0,0,0", Scale="0,0,0") {
    alphaInputType=Bool
    rotationSpace=BCS_BoneSpace
    translationMode=BMM_Additive
    alphaBoolBlend {blendInTime=0.25, blendOption=Linear, blendOutTime=0.25}
    alphaScaleBiasClamp {bInterpResult=true, interpSpeedDecreasing=5, interpSpeedIncreasing=5}
    boneToModify {boneName=axle_pivot_fr_l}
    scale=<0 0 0>
    translation {z=-40}
   }
   n26 = "Transform (Modify) Bone - Bone: frame"!(ComponentPose=@n25, bAlphaBoolEnabled:bool=@n23, Translation="0,0,20", Rotation="0,0,10", Scale="0,0,0") -> (Pose) {
    alphaInputType=Bool
    rotationMode=BMM_Additive
    rotationSpace=BCS_BoneSpace
    translationMode=BMM_Additive
    translationSpace=BCS_BoneSpace
    alphaBoolBlend {blendInTime=0.25, blendOption=Linear, blendOutTime=0.25}
    alphaScaleBiasClamp {bInterpResult=true, interpSpeedDecreasing=5, interpSpeedIncreasing=5}
    boneToModify {boneName=frame}
    rotation {roll=10}
    scale=<0 0 0>
    translation {z=20}
   }
   n27 = Get IsTiltLeft() -> (IsTiltLeft:bool)
   n28 = "Transform (Modify) Bone - Bone: axle_pivot_bk_r"(ComponentPose=@n26, bAlphaBoolEnabled=@n27, Translation="0,0,-40", Rotation="0,0,0", Scale="0,0,0") {
    alphaInputType=Bool
    rotationSpace=BCS_BoneSpace
    translationMode=BMM_Additive
    alphaBoolBlend {blendInTime=0.25, blendOption=Linear, blendOutTime=0.25}
    alphaScaleBiasClamp {bInterpResult=true, interpSpeedDecreasing=5, interpSpeedIncreasing=5}
    boneToModify {boneName=axle_pivot_bk_r}
    scale=<0 0 0>
    translation {z=-40}
   }
   n29 = "Transform (Modify) Bone - Bone: axle_pivot_fr_r"(ComponentPose=@n28, bAlphaBoolEnabled=@n27, Translation="0,0,-40", Rotation="0,0,0", Scale="0,0,0") {
    alphaInputType=Bool
    rotationSpace=BCS_BoneSpace
    translationMode=BMM_Additive
    alphaBoolBlend {blendInTime=0.25, blendOption=Linear, blendOutTime=0.25}
    alphaScaleBiasClamp {bInterpResult=true, interpSpeedDecreasing=5, interpSpeedIncreasing=5}
    boneToModify {boneName=axle_pivot_fr_r}
    scale=<0 0 0>
    translation {z=-40}
   }
   n30 = "Transform (Modify) Bone - Bone: frame"(ComponentPose=@n29, bAlphaBoolEnabled=@n27, Translation="0,0,20", Rotation="0,0,-10", Scale="0,0,0") {
    alphaInputType=Bool
    rotationMode=BMM_Additive
    rotationSpace=BCS_BoneSpace
    translationMode=BMM_Additive
    translationSpace=BCS_BoneSpace
    alphaBoolBlend {blendInTime=0.25, blendOption=Linear, blendOutTime=0.25}
    alphaScaleBiasClamp {bInterpResult=true, interpSpeedDecreasing=5, interpSpeedIncreasing=5}
    boneToModify {boneName=frame}
    rotation {roll=-10}
    scale=<0 0 0>
    translation {z=20}
   }
   n31 = "Transform (Modify) Bone - Bone: body"!(
    ComponentPose=@n30
    bAlphaBoolEnabled:bool=Get IsDroppedLeft() -> (IsDroppedLeft:bool)
    Translation="0,0,-5"
    Rotation="0,0,-5"
    Scale="0,0,0"
   ) -> (Pose) {
    alphaInputType=Bool
    rotationMode=BMM_Additive
    rotationSpace=BCS_BoneSpace
    translationMode=BMM_Additive
    translationSpace=BCS_BoneSpace
    alphaBoolBlend {blendInTime=0.25, blendOption=Linear, blendOutTime=0.25}
    alphaScaleBiasClamp {bInterpResult=true, interpSpeedDecreasing=5, interpSpeedIncreasing=5}
    boneToModify {boneName=body}
    rotation {roll=-5}
    scale=<0 0 0>
    translation {z=-5}
   }
   n32 = "Transform (Modify) Bone - Bone: body"(ComponentPose=@n31, bAlphaBoolEnabled=Get IsDroppedRight() -> (IsDroppedRight:bool), Translation="0,0,-5", Rotation="0,0,5", Scale="0,0,0") {
    alphaInputType=Bool
    rotationMode=BMM_Additive
    rotationSpace=BCS_BoneSpace
    translationMode=BMM_Additive
    translationSpace=BCS_BoneSpace
    alphaBoolBlend {blendInTime=0.25, blendOption=Linear, blendOutTime=0.25}
    alphaScaleBiasClamp {bInterpResult=true, interpSpeedDecreasing=5, interpSpeedIncreasing=5}
    boneToModify {boneName=body}
    rotation {roll=5}
    scale=<0 0 0>
    translation {z=-5}
   }
   n33 = Get IsBouncing() -> (IsBouncing:bool)
   n34 = "Transform (Modify) Bone - Bone: body"(ComponentPose=@n32, bAlphaBoolEnabled=@n33, Translation="0,0,20", Rotation="10,0,0", Scale="0,0,0") {
    alphaInputType=Bool
    rotationMode=BMM_Additive
    rotationSpace=BCS_BoneSpace
    translationMode=BMM_Additive
    translationSpace=BCS_BoneSpace
    alphaBoolBlend {blendInTime=0.4, blendOption=CircularInOut, blendOutTime=0.4}
    alphaScaleBiasClamp {bInterpResult=true, interpSpeedDecreasing=5, interpSpeedIncreasing=5}
    boneToModify {boneName=body}
    rotation {pitch=10}
    scale=<0 0 0>
    translation {z=20}
   }
   n35 = "Transform (Modify) Bone - Bone: front_differential"!(ComponentPose=@n34, bAlphaBoolEnabled:bool=@n33, Translation="0,0,25", Rotation="0,0,0", Scale="0,0,0") -> (Pose) {
    alphaInputType=Bool
    rotationMode=BMM_Additive
    rotationSpace=BCS_BoneSpace
    translationMode=BMM_Additive
    translationSpace=BCS_BoneSpace
    alphaBoolBlend {blendInTime=0.4, blendOption=CircularInOut, blendOutTime=0.4}
    alphaScaleBiasClamp {bInterpResult=true, interpSpeedDecreasing=5, interpSpeedIncreasing=5}
    boneToModify {boneName=front_differential}
    scale=<0 0 0>
    translation {z=25}
   }
   Output Pose(Result=Component To Local(ComponentPose=@n35)) {layerGroup=None}
  ]}
 EventGraph {nodes [
   n1 = StopIt() {is=title K2Node_CustomEvent}
   > Set IsDroppedLeft(IsDroppedLeft:bool=false) -> (Output_Get:bool)
   > Set IsDroppedRight(IsDroppedRight:bool=false) -> (Output_Get:bool)
   n2 = DropItLeft() {is=title K2Node_CustomEvent}
   > n3 = Set IsDroppedLeft(IsDroppedLeft=true)
   n4 = DropItRight() {is=title K2Node_CustomEvent}
   Set Timer by Event(execute=@n3.then, Delegate=@n4, Time=0.5, bLooping=false)
   n5 = Start Bounce() {is=title K2Node_CustomEvent}
   > n6 = Set IsBouncing(IsBouncing:bool=true) -> (Output_Get:bool)
   n7 = StopBounce() {is=title K2Node_CustomEvent}
   Set Timer by Event(execute=@n6.then, Delegate=@n7, Time=0.5, bLooping=false)
   Event BlueprintUpdateAnimation()
   > n8 = Sequence!(execute) -> (then_0, then_1, then_2:Exec)
   n9 = GetSteeringAngle!(self:Fort Athena Vehicle Object Reference=Get Vehicle() -> (Vehicle:Fort Athena SKVehicle Object Reference)) -> (ReturnValue:float)
   > n10 = Branch(Condition="float > float"(A=@n9, B=5))
   > Set IsTiltLeft(IsTiltLeft:bool=true) -> (Output_Get:bool)
   Set IsTiltLeft(execute=@n10.else, IsTiltLeft=false)
   n11 = Branch(execute=@n8.then_1, Condition="float < float"(A=@n9, B=-5))
   > Set IsTiltRight(IsTiltRight:bool=true) -> (Output_Get:bool)
   Set IsTiltRight(execute=@n11.else, IsTiltRight=false)
   n12 = Get IsTiltLeft()
   n13 = Get IsTiltRight()
   n14 = Branch(execute=@n8.then_2, Condition=AND Boolean(A=NOT Boolean(A=@n12), B=NOT Boolean(A=@n13)))
   > Set Is Idle(Is Idle:bool=true) -> (Output_Get:bool)
   Set Is Idle(execute=@n14.else, Is Idle=false)
   Event BlueprintBeginPlay() {is=title K2Node_Event}
   > n15 = Cast To Clyde_Valet_Lowrider_Vehicle!(execute, Object=TryGetPawnOwner()) -> (then, CastFailed, AsClyde Valet Lowrider Vehicle:Clyde Valet Lowrider Vehicle Object Reference) {is=title K2Node_DynamicCast}
   > Is Valid(InputObject=@n15)
   > n16 = Sequence()
   > Bind Event to Start Bounce(self:Clyde Valet Lowrider Vehicle Object Reference=@n15, Delegate=@n5) {is=title K2Node_AddDelegate}
   Bind Event to Stop It(execute=@n16.then_1, self:Clyde Valet Lowrider Vehicle Object Reference=@n15, Delegate=@n1) {is=title K2Node_AddDelegate}
   Bind Event to Drop It Left(execute=@n16.then_2, self:Clyde Valet Lowrider Vehicle Object Reference=@n15, Delegate=@n2) {is=title K2Node_AddDelegate}
   Set IsBouncing(execute=@n7.then, IsBouncing=false)
   Set IsDroppedRight(execute=@n4.then, IsDroppedRight=true)
  ]}
]
variables [
 Is Idle {type=boolean, default=true}
 IsBouncing {type=boolean}
 IsDroppedLeft {type=boolean}
 IsDroppedRight {type=boolean}
 IsTiltLeft {type=boolean}
 IsTiltRight {type=boolean}
]
